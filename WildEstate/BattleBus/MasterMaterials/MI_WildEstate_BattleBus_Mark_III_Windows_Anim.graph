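# Window material for the Wild Estate Battle Bus, with fake reflections and an Otto figure on a virtual plane.
# Time-of-day glass and reflection colors, shaded by sun direction, feed eye-adaptation-corrected emissive.
# A line texture, shifted by instance position and view direction, supplies window lines; Otto uses VirtualPlaneCoordinates with a sphere mask.
# Static switches control party-bus animation WPO and optional Fortnite hit-effect wobble.
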Material /WildEstate/BattleBus/MasterMaterials/MI_WildEstate_BattleBus_Mark_III_Windows_Anim
graph {
 properties {bUseMaterialAttributes=true, pixelDepthOffsetMode=PDOM_Legacy, usedShadingModels=Default Lit}
 nodes [
  n1 = TextureCoordinate()
  n2 = ComponentMask(None=VectorParameter() {group=Otto_Card, parameterName=Plane_00_UVScale, defaultValue=<rgba 512 512 0 1>}.RGB) {g=true, r=true}
  Add(A=Multiply(A=@n1, B=@n2), B=@n1)
  ComponentMask(None=VectorParameter() {group=Otto_Card, parameterName=Plane_00_UVOffset, defaultValue=<rgba 0 0 0 1>}.RGB) {g=true, r=true}
  n3 = VectorParameter() {group=Otto_Card, parameterName=Plane_00_UVOffset, defaultValue=<rgba 0 0 0 1>}
  ComponentMask(None=@n3.RGB) {g=true, r=true}
  Constant() {r=0.5}
  n4 = PreSkinnedPosition()
  n5 = VertexInterpolator(VS=@n4)
  n6 = PreSkinnedLocalBounds()
  Divide(A=Subtract(A=ShaderStageSwitch(PixelShader=@n5.PS, VertexShader=@n4), B=@n6.Min), B=Subtract(A=@n6.Max, B=@n6.Max))
  n7 = VectorParameter() {parameterName=Reflection_Color_Morning, defaultValue=<rgba 0.724 0.6384 0.4014 1>}
  n8 = VectorParameter() {parameterName=Reflection_Color_Day, defaultValue=<rgba 0.1131 0.724 0.6119 1>}
  n9 = VectorParameter() {parameterName=Reflection_Color_Evening, defaultValue=<rgba 0.1131 0.724 0.6119 1>}
  n10 = VectorParameter() {parameterName=Reflection_Color_Night, defaultValue=<rgba 0.0254 0.1028 0.4427 1>}
  n11 = NamedRerouteUsage()
  n12 = Add(
   A=MaterialFunctionCall(
    Color_Morning=VectorParameter() {parameterName=Glass_Color_Morning, defaultValue=<rgba 0.8229 0.4438 0.1029 1>}.RGB
    Color_Day=VectorParameter() {parameterName=Glass_Color_Day, defaultValue=<rgba 0.1131 0.724 0.6119 1>}.RGB
    Color_Evening=VectorParameter() {parameterName=Glass_Color_Evening, defaultValue=<rgba 0.724 0.381 0.131 1>}.RGB
    Color_Night=VectorParameter() {parameterName=Glass_Color_Night, defaultValue=<rgba 0.0407 0.1246 0.224 1>}.RGB
   ) {materialFunction=&Color_Time_of_Day}.Result
   B=Multiply(
    A=MaterialFunctionCall(Color_Morning=@n7.RGB, Color_Day=@n8.RGB, Color_Evening=@n9.RGB, Color_Night=@n10.RGB) {materialFunction=&Color_Time_of_Day}.Result
    B=OneMinus(None=@n11)
   )
  )
  n13 = Multiply(A=NamedRerouteUsage(), B=NamedRerouteUsage())
  LinearInterpolate(A=@n12, B=LinearInterpolate(A=@n13, B=Add(A=@n13, B=@n12)) {constAlpha=0.1}, Alpha=Saturate(None=@n13))
  NamedRerouteDeclaration(None=LinearInterpolate(
    A=ScalarParameter() {defaultValue=1, parameterName=Opacity_Lines}
    B=ScalarParameter() {defaultValue=1, parameterName=Opacity_Glass}
    Alpha=NamedRerouteUsage()
   )) {name=Opacity, nodeColor=<rgba 0.4531 1 0 1>}
  n14 = TransformPosition(None=Constant3Vector()) {transformSourceType=TRANSFORMPOSSOURCE_Instance}
  n15 = MaterialFunctionCall(Float3=Divide(A=@n14) {constB=100}) {materialFunction=&BreakOutFloat3Components}
  n16 = Add(A=Add(A=@n15.R, B=@n15.R), B=@n15.B)
  n17 = Add(
   A=Multiply(A=Frac(None=@n16)) {constB=33.3}
   B=Multiply(A=TextureCoordinate(), B=ScalarParameter() {defaultValue=0.5, parameterName=GlassRelectionTiling})
  )
  n18 = Normalize(VectorInput=Subtract(A=@n14, B=CameraPositionWS()))
  n19 = Multiply(
   A=ComponentMask(None=Transform(None=@n18) {transformSourceType=TRANSFORMSOURCE_Local, transformType=TRANSFORM_Tangent}) {g=true, r=true}
   B=ScalarParameter() {defaultValue=0.1, parameterName=GlassRelectionOffset}
  )
  NamedRerouteDeclaration(None=TextureSample(UVs=Add(A=@n17, B=@n19)) {texture=/WildEstate/BattleBus/Textures/T_WildEstate_PartyBus_Windows}.R) {name=WindowLines, nodeColor=<rgba 1 0 0.7734 1>}
  n20 = CollectionParameter() {
   collection=&FortniteMaterialParameters
   parameterName=SunlightDirectionalVector
  }
  n21 = Multiply(A=VertexNormalWS(), B=TwoSidedSign())
  n22 = Saturate(None=Max(A=DotProduct(A=@n20, B=@n21), B=ScalarParameter() {defaultValue=0.4109, parameterName=MinDarkness}))
  NamedRerouteDeclaration(None=EyeAdaptationInverse(LightValueInput=Multiply(A=@n12, B=@n22)).EyeAdaptationInverse) {name=Emissive, nodeColor=<rgba 0 1 0.9062 1>}
  n23 = PreSkinnedPosition()
  n24 = ShaderStageSwitch(PixelShader=VertexInterpolator(VS=@n23).PS, VertexShader=@n23)
  n25 = PreSkinnedLocalBounds()
  n26 = Add(
   A=Divide(A=Subtract(A=@n24, B=@n25.Min), B=Subtract(A=@n25.Max, B=@n25.Max))
   B=ScalarParameter() {defaultValue=-0.2, group=Otto_Card, parameterName=Otto_FrontMaskOffset}
  ) {constB=-0.2}
  n27 = ScalarParameter() {defaultValue=3, group=Otto_Card, parameterName=Otto_FrontMaskContrast}
  NamedRerouteDeclaration(None=ComponentMask(None=Power(Base=@n26, Exp=@n27)) {g=true}) {name=CalculatedBounds, nodeColor=<rgba 0.9453 1 0 1>}
  n28 = VectorParameter() {group=Otto_Card, parameterName=Plane_00_Normal, defaultValue=<rgba 1 0 0 1>}
  n29 = Normalize(VectorInput=Subtract(A=Normalize(VectorInput=@n28.RGB), B=CameraPositionWS()))
  n30 = Constant3Vector() {constant=<rgba 0 0 -1 0>}
  n31 = Add(A=VectorParameter() {group=Otto_Card, parameterName=Plane_00_Center, defaultValue=<rgba 0 0 0 1>}.RGB, B=@n3.RGB)
  n32 = MaterialFunctionCall(
   Plane Normal Axis=@n29
   Plane X Axis=Normalize(VectorInput=CrossProduct(A=@n29, B=@n30))
   Plane Center=Add(A=TransformPosition(None=@n31) {transformType=TRANSFORMPOSSOURCE_World}, B=NamedRerouteUsage())
   UV Scale=@n2
  ) {materialFunction=&VirtualPlaneCoordinates}
  n33 = TextureSample(UVs=@n32.UVs) {samplerType=SAMPLERTYPE_Masks, texture=/WildEstate/BattleBus/MasterMaterials/Textures/T_WildEstate_BBOtto}
  n34 = SphereMask(
   A=@n32.UVs
   B=Constant2Vector() {g=0.5, r=0.5}
   Radius=ScalarParameter() {defaultValue=230, group=Otto_Card, parameterName=Plane_00_SphereMaskRadius}
   Hardness=ScalarParameter() {group=Otto_Card, parameterName=Plane_00_SphereMaskHardness}
  ) {attenuationRadius=0.5}
  NamedRerouteDeclaration(None=Multiply(A=Multiply(A=@n33.R, B=@n34), B=ScalarParameter() {defaultValue=4.8552, group=Otto_Card, parameterName=Otto_Intensity})) {name=OttoMan, nodeColor=<rgba 0.5469 1 0 1>}
  n35 = MaterialFunctionCall() {materialFunction=/Game/Athena/Aircraft/PartyBus_RT/Materials/MF_PartyBusAnimationWPO}
  NamedRerouteDeclaration(None=@n35.Result) {name=WPO, nodeColor=<rgba 1 0 0.1406 1>}
  OneMinus(None=Fresnel(ExponentIn=ScalarParameter() {defaultValue=1.4086, parameterName=FillFresnelEdge}, Normal=VertexNormalWS()) {baseReflectFraction=0})
  n36 = SetMaterialAttributes!(MaterialAttributes, Emissive Color=NamedRerouteUsage(), Opacity=NamedRerouteUsage()) {attributeSetTypes [#1, #2]}
  n37 = StaticSwitchParameter(
   True=@n36
   False=SetMaterialAttributes!(MaterialAttributes=@n36, World Position Offset=@n35.Result) {attributeSetTypes [#3]}
  ) {parameterName=UseDisableAllWPO}
  n38 = GetMaterialAttributes(None=@n37) {attributeGetTypes [#3]}
  n39 = MaterialFunctionCall!(
   WSDeformationCenter=VectorParameter() {bUseCustomPrimitiveData=true, group=z_code, parameterName=WSDeformationCenter, primitiveDataIndex=20}.RGB
   Code_DestructionAmount=ScalarParameter() {group=z_code, parameterName=Code_DestructionAmount}
   Code_UseUvForDamage
   Code_UseHitEffects=StaticBoolParameter() {defaultValue=true, group=z_code, parameterName=Code_UseHitEffects}
   Code_Emissive
   WSDamageRadius=ScalarParameter() {bUseCustomPrimitiveData=true, defaultValue=1, group=z_code, parameterName=WSDamageRadius, primitiveDataIndex=26}
   CrackDamageShader
   Diffuse
   ScalarModDamageCracks
   ScalarOffsetforDamageCracks
   ExposeLerpFunctionForInvertingTheHitDistanceCalc
   InvertDamageFalloff
   CrackWorldPosition
   WorldPosition
   DistanceCullFadeAnimationOffset=ScalarParameter() {bUseCustomPrimitiveData=true, group=z_code, parameterName=DistanceCullFadeAnimationOffset, primitiveDataIndex=12}
  ) {materialFunction=/Game/Packages/jl_tests/Fortnite_HitEffects}
  StaticSwitchParameter(
   True=SetMaterialAttributes(MaterialAttributes=@n38.MaterialAttributes, World Position Offset=Add(A=@n38.World Position Offset, B=@n39.Hit And Shrink WPO)) {attributeSetTypes [#3]}
   False=@n37
  ) {parameterName=UseDamageWobble}
  TransformPosition(None=@n17) {transformSourceType=TRANSFORMPOSSOURCE_World}
 ]
}
