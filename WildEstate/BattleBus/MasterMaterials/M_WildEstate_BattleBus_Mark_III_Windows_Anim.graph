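# Master material for the animated Battle Bus Mark III windows, faking lit glass with an Otto figure inside.
# Emissive blends time-of-day glass and reflection colors with parallax bump-offset, sphere-masked Otto cards (quality-switched), scaled by sunlight and inverse eye adaptation.
# World-aligned window-line masks feed reroutes, including an opacity blend.
# Party bus animation drives world position offset, with static switches for disabling it or adding hit-effect damage wobble.
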Material /WildEstate/BattleBus/MasterMaterials/M_WildEstate_BattleBus_Mark_III_Windows_Anim
graph {
 properties {bUseMaterialAttributes=true, pixelDepthOffsetMode=PDOM_Legacy, usedShadingModels=Default Lit}
 nodes [
  n1 = TextureCoordinate()
  BumpOffset(
   Coordinate=@n1
   Height=ScalarParameter() {defaultValue=-2.5, group=Otto_Card, parameterName=Plane_00_BumpOffset}
   HeightRatioInput=Constant() {r=0.05}
  )
  n2 = TextureCoordinate()
  BumpOffset(
   Coordinate=@n2
   Height=ScalarParameter() {defaultValue=-2.5, group=Otto_Card, parameterName=Plane_01_BumpOffset}
   HeightRatioInput=Constant() {r=0.05}
  )
  n3 = TextureCoordinate()
  BumpOffset(
   Coordinate=@n3
   Height=ScalarParameter() {defaultValue=-2.5, group=Otto_Card, parameterName=Plane_02_BumpOffset}
   HeightRatioInput=Constant() {r=0.05}
  )
  n4 = PreSkinnedPosition()
  n5 = PreSkinnedLocalBounds()
  n6 = Subtract(A=ShaderStageSwitch(PixelShader=VertexInterpolator(VS=@n4).PS, VertexShader=@n4), B=@n5.Min)
  n7 = Subtract(A=@n5.Max, B=@n5.Max)
  ComponentMask(None=Transform(None=Divide(A=@n6, B=@n7)) {transformSourceType=TRANSFORMSOURCE_World, transformType=TRANSFORM_Local}) {g=true, r=true}
  n8 = VectorParameter() {group=Otto_Card, parameterName=Plane_00_UVOffset, defaultValue=<rgba 0 0 0 1>}
  ComponentMask(None=@n8.RGB) {g=true, r=true}
  Constant()
  n9 = PreSkinnedPosition()
  n10 = VertexInterpolator(VS=@n9)
  n11 = PreSkinnedLocalBounds()
  Divide(A=Subtract(A=ShaderStageSwitch(PixelShader=@n10.PS, VertexShader=@n9), B=@n11.Min), B=Subtract(A=@n11.Max, B=@n11.Max))
  n12 = ObjectLocalBounds()
  Divide(A=Subtract(B=@n12.Min), B=Subtract(A=@n12.Max, B=@n12.Max))
  n13 = TextureObject() {texture=/WildEstate/BattleBus/Textures/T_WildEstate_PartyBus_Windows}
  n14 = WorldPosition()
  n15 = ObjectLocalBounds()
  n16 = ObjectBounds()
  MaterialFunctionCall(
   TextureObject=@n13
   TextureSize=ScalarParameter() {defaultValue=0.5, parameterName=GlassRelectionTiling}
   WorldPosition=Divide(A=TransformPosition(None=@n14.XYZ) {transformSourceType=TRANSFORMPOSSOURCE_World}, B=Divide(A=@n15.Extents, B=@n16))
   World Space Normal=VertexInterpolator(VS=PreSkinnedNormal()).PS
  ) {materialFunction=&WorldAlignedTexture}
  n17 = VectorParameter() {group=Otto_Card, parameterName=Plane_00_Normal, defaultValue=<rgba 1 0 0 1>}
  n18 = Normalize(VectorInput=Subtract(A=Normalize(VectorInput=@n17.RGB), B=CameraPositionWS()))
  n19 = Constant3Vector() {constant=<rgba 0 0 -1 0>}
  n20 = Add(A=VectorParameter() {group=Otto_Card, parameterName=Plane_00_Center, defaultValue=<rgba 0 0 0 1>}.RGB, B=@n8.RGB)
  MaterialFunctionCall(
   Plane Normal Axis=@n18
   Plane X Axis=Normalize(VectorInput=CrossProduct(A=@n18, B=@n19))
   Plane Center=Add(A=TransformPosition(None=@n20) {transformType=TRANSFORMPOSSOURCE_World}, B=NamedRerouteUsage())
   UV Scale=ComponentMask(None=VectorParameter() {group=Otto_Card, parameterName=Plane_00_UVScale, defaultValue=<rgba 1 1 0 1>}.RGB) {g=true, r=true}
  ) {materialFunction=&VirtualPlaneCoordinates}
  MaterialFunctionCall!(
   CameraVector=Transform(None=CameraVectorWS()) {transformSourceType=TRANSFORMSOURCE_World, transformType=TRANSFORM_Local}
   Texture=TextureObject() {texture=&T_WildEstate_PartyBus_Windows}
   Tiling=ScalarParameter() {defaultValue=0.5, parameterName=GlassRelectionTiling}
   Offset
   AxisFadeContrast=Constant() {r=24}
  ) {materialFunction=/Engine/Functions/Engine_MaterialFunctions01/Texturing/TriplanarCameraVector}
  Multiply(A=TextureCoordinate(), B=ScalarParameter() {defaultValue=0.5, parameterName=GlassRelectionTiling})
  n21 = MaterialFunctionCall(
   Color_Morning=VectorParameter() {parameterName=Reflection_Color_Morning, defaultValue=<rgba 0.1964 0.3241 0.4375 1>}.RGB
   Color_Day=VectorParameter() {parameterName=Reflection_Color_Day, defaultValue=<rgba 0.1131 0.724 0.6119 1>}.RGB
   Color_Evening=VectorParameter() {parameterName=Reflection_Color_Evening, defaultValue=<rgba 0.1964 0.3241 0.4375 1>}.RGB
   Color_Night=VectorParameter() {parameterName=Reflection_Color_Night, defaultValue=<rgba 0.0371 0.258 0.276 1>}.RGB
  ) {materialFunction=&Color_Time_of_Day}
  Multiply(A=@n21.Result) {constB=100}
  Multiply(A=TextureCoordinate(), B=ScalarParameter() {defaultValue=0.5, group=Glass, parameterName=GlassRelectionTilingUV}) {constB=0.5}
  NamedRerouteDeclaration(None=LinearInterpolate(
    A=ScalarParameter() {defaultValue=1, parameterName=Opacity_Lines}
    B=ScalarParameter() {defaultValue=1, parameterName=Opacity_Glass}
    Alpha=NamedRerouteUsage()
   )) {name=Opacity, nodeColor=<rgba 0.4531 1 0 1>}
  n22 = Constant() {r=364}
  n23 = PreSkinnedPosition()
  n24 = VertexInterpolator(VS=@n23)
  n25 = StaticBoolParameter() {defaultValue=true, group=z_GameplayParameters_DestructibleDamage, parameterName=UseLocalSpace, sortPriority=100}
  n26 = StaticSwitch(
   True=Multiply(
    A=MaterialFunctionCall() {materialFunction=&ObjectScale}.Scale XYZ
    B=ShaderStageSwitch(PixelShader=@n24.PS, VertexShader=@n23)
   )
   False=WorldPosition().XYZ
   Value=@n25
  )
  n27 = PreSkinnedNormal()
  n28 = StaticSwitch(True=ShaderStageSwitch(PixelShader=VertexInterpolator(VS=@n27).PS, VertexShader=@n27), False=VertexNormalWS(), Value=@n25)
  NamedRerouteDeclaration(None=ComponentMask(None=MaterialFunctionCall(TextureObject=@n13, TextureSize=@n22, WorldPosition=@n26, World Space Normal=@n28) {materialFunction=&WorldAlignedTexture}.XYZ Texture) {g=true}) {name=WindowLines, nodeColor=<rgba 1 0 0.7734 1>}
  n29 = Transform(None=VertexNormalWS()) {transformSourceType=TRANSFORMSOURCE_World, transformType=TRANSFORM_Local}
  n30 = MaterialFunctionCall(Float3=Abs(None=Normalize(VectorInput=@n29))) {materialFunction=&BreakOutFloat3Components}
  n31 = TransformPosition(None=WorldPosition().XYZ) {transformSourceType=TRANSFORMPOSSOURCE_World}
  n32 = ComponentMask(None=@n31) {b=true, desc=YZ, g=true}
  n33 = ComponentMask(None=@n31) {b=true, desc=XZ, r=true}
  n34 = Divide(
   A=If(
    A=@n30.B
    B=Max(A=@n30.R, B=@n30.R)
    "A > B"=ComponentMask(None=@n31) {desc=YX, g=true, r=true}
    "A < B"=If(A=@n30.R, B=@n30.R, "A > B"=@n32, "A < B"=@n33)
   )
   B=ScalarParameter() {defaultValue=512, parameterName=GlassRelectionTilingTriPlanar}
  )
  NamedRerouteDeclaration(None=@n34) {name=CheapTriplanar, nodeColor=<rgba 0 0.7891 1 1>}
  n35 = VectorParameter() {group=Otto_Card, parameterName=Otto_01_UVOffset, defaultValue=<rgba 0 0 0 1>}
  n36 = BumpOffset(
   Coordinate=TextureCoordinate()
   Height=ScalarParameter() {defaultValue=-2.5, group=Otto_Card, parameterName=Otto_01_BumpOffset}
   HeightRatioInput=Constant() {r=0.05}
  )
  n37 = ComponentMask(None=VectorParameter() {group=Otto_Card, parameterName=Otto_01_UVScale, defaultValue=<rgba 1 1 0 1>}.RGB) {g=true, r=true}
  NamedRerouteDeclaration(None=Add(A=ComponentMask(None=@n35.RGB) {g=true, r=true}, B=Multiply(A=@n36, B=@n37))) {name=Otto01UV, nodeColor=<rgba 1 0 0.1641 1>}
  n38 = VectorParameter() {group=Otto_Card, parameterName=Otto_00_UVOffset, defaultValue=<rgba 0 0 0 1>}
  n39 = BumpOffset(
   Coordinate=TextureCoordinate()
   Height=ScalarParameter() {defaultValue=-2.5, group=Otto_Card, parameterName=Otto_00_BumpOffset}
   HeightRatioInput=Constant() {r=0.05}
  )
  n40 = ComponentMask(None=VectorParameter() {group=Otto_Card, parameterName=Otto_00_UVScale, defaultValue=<rgba 1 1 0 1>}.RGB) {g=true, r=true}
  NamedRerouteDeclaration(None=Add(A=ComponentMask(None=@n38.RGB) {g=true, r=true}, B=Multiply(A=@n39, B=@n40))) {name=Otto00UV, nodeColor=<rgba 1 0 0.1641 1>}
  n41 = VectorParameter() {group=Otto_Card, parameterName=Plane_02_UVOffset, defaultValue=<rgba 0 0 0 1>}
  n42 = ComponentMask(None=VectorParameter() {group=Otto_Card, parameterName=Plane_02_UVScale, defaultValue=<rgba 1 1 0 1>}.RGB) {g=true, r=true}
  NamedRerouteDeclaration(None=Add(A=ComponentMask(None=@n41.RGB) {g=true, r=true}, B=Multiply(A=@n3, B=@n42))) {name=Plane02UV, nodeColor=<rgba 0.875 0 1 1>}
  n43 = VectorParameter() {group=Otto_Card, parameterName=Plane_01_UVOffset, defaultValue=<rgba 0 0 0 1>}
  n44 = ComponentMask(None=VectorParameter() {group=Otto_Card, parameterName=Plane_01_UVScale, defaultValue=<rgba 1 1 0 1>}.RGB) {g=true, r=true}
  NamedRerouteDeclaration(None=Add(A=ComponentMask(None=@n43.RGB) {g=true, r=true}, B=Multiply(A=@n2, B=@n44))) {name=Plane01UV, nodeColor=<rgba 0.875 0 1 1>}
  n45 = VectorParameter() {group=Otto_Card, parameterName=Plane_00_UVOffset, defaultValue=<rgba 0 0 0 1>}
  n46 = ComponentMask(None=VectorParameter() {group=Otto_Card, parameterName=Plane_00_UVScale, defaultValue=<rgba 1 1 0 1>}.RGB) {g=true, r=true}
  NamedRerouteDeclaration(None=Add(A=ComponentMask(None=@n45.RGB) {g=true, r=true}, B=Multiply(A=@n1, B=@n46))) {name=Plane00UV, nodeColor=<rgba 0.875 0 1 1>}
  NamedRerouteDeclaration(None=TextureObject() {samplerType=SAMPLERTYPE_Masks, texture=/WildEstate/BattleBus/MasterMaterials/Textures/T_WildEstate_BBOtto}) {name=OttoTex, nodeColor=<rgba 0 0.0859 1 1>}
  n47 = TextureSample(UVs=NamedRerouteUsage(), Tex=NamedRerouteUsage()) {
   samplerSource=SSM_Clamp_WorldGroupSettings
   samplerType=SAMPLERTYPE_Masks
   texture=&T_WildEstate_BBOtto
  }
  n48 = NamedRerouteUsage()
  n49 = Constant() {r=0.5}
  n50 = ScalarParameter() {group=Otto_Card, parameterName=Plane_00_SphereMaskRadius}
  n51 = ScalarParameter() {group=Otto_Card, parameterName=Plane_00_SphereMaskHardness}
  NamedRerouteDeclaration(None=Multiply(A=@n47.R, B=SphereMask(A=@n48, B=@n49, Radius=@n50, Hardness=@n51) {attenuationRadius=0.5})) {name=O_01, nodeColor=<rgba 0.1719 1 0 1>}
  n52 = TextureSample(UVs=NamedRerouteUsage(), Tex=NamedRerouteUsage()) {
   samplerSource=SSM_Clamp_WorldGroupSettings
   samplerType=SAMPLERTYPE_Masks
   texture=&T_WildEstate_BBOtto
  }
  n53 = NamedRerouteUsage()
  n54 = Constant() {r=0.5}
  n55 = ScalarParameter() {group=Otto_Card, parameterName=Plane_00_SphereMaskRadius}
  n56 = ScalarParameter() {group=Otto_Card, parameterName=Plane_00_SphereMaskHardness}
  NamedRerouteDeclaration(None=Multiply(A=@n52.R, B=SphereMask(A=@n53, B=@n54, Radius=@n55, Hardness=@n56) {attenuationRadius=0.5})) {name=O_02, nodeColor=<rgba 1 0 0.4922 1>}
  n57 = MaterialFunctionCall(
   Color_Morning=VectorParameter() {parameterName=Glass_Color_Morning, defaultValue=<rgba 0.1747 0.2248 0.55 1>}.RGB
   Color_Day=VectorParameter() {parameterName=Glass_Color_Day, defaultValue=<rgba 0.1276 0.7 0.7 1>}.RGB
   Color_Evening=VectorParameter() {parameterName=Glass_Color_Evening, defaultValue=<rgba 0.125 0.2124 0.5 1>}.RGB
   Color_Night=VectorParameter() {parameterName=Glass_Color_Night, defaultValue=<rgba 0.0198 0.3849 0.4 1>}.RGB
  ) {materialFunction=&Color_Time_of_Day}
  n58 = Multiply(A=@n21.Result, B=OneMinus(None=NamedRerouteUsage()))
  n59 = Constant() {r=0.2}
  n60 = Constant() {r=0.33}
  n61 = Constant() {r=0.1}
  n62 = LinearInterpolate(
   A=Multiply(
    A=@n57.Result
    B=MaterialFunctionCall(Value_Morning=@n59, Value_Day=@n60, Value_Evening=@n61, Value_Night=@n61) {materialFunction=&Scalar_Time_of_Day}.Result
   )
   B=@n21.Result
  )
  n63 = Multiply(A=NamedRerouteUsage(), B=NamedRerouteUsage())
  n64 = LinearInterpolate(
   A=Add(A=@n57.Result, B=@n58)
   B=Add(A=LinearInterpolate(A=@n57.Result, B=@n62, Alpha=@n63), B=@n58)
   Alpha=ScalarParameter() {group=Otto_Card, parameterName=Use_Otto}
  )
  n65 = CollectionParameter() {
   collection=&FortniteMaterialParameters
   parameterName=SunlightDirectionalVector
  }
  n66 = Multiply(A=VertexNormalWS(), B=TwoSidedSign())
  n67 = Saturate(None=Max(A=DotProduct(A=@n65, B=@n66), B=ScalarParameter() {defaultValue=0.4109, parameterName=MinDarkness}))
  NamedRerouteDeclaration(None=EyeAdaptationInverse(LightValueInput=Multiply(A=@n64, B=@n67)).EyeAdaptationInverse) {name=Emissive, nodeColor=<rgba 0 1 0.9062 1>}
  n68 = PreSkinnedPosition()
  n69 = VertexInterpolator(VS=@n68)
  n70 = PreSkinnedLocalBounds()
  n71 = Divide(A=Subtract(A=ShaderStageSwitch(PixelShader=@n69.PS, VertexShader=@n68), B=@n70.Min), B=Subtract(A=@n70.Max, B=@n70.Max))
  n72 = Add(A=@n71, B=ScalarParameter() {defaultValue=-0.2, group=Otto_Card, parameterName=Otto_FrontMaskOffset}) {constB=-0.2}
  n73 = ScalarParameter() {defaultValue=4.552, group=Otto_Card, parameterName=Otto_FrontMaskContrast}
  NamedRerouteDeclaration(None=ComponentMask(None=Power(Base=@n72, Exp=@n73)) {g=true}) {name=CalculatedBounds, nodeColor=<rgba 0.9453 1 0 1>}
  n74 = NamedRerouteUsage()
  n75 = NamedRerouteUsage()
  n76 = TextureSample(UVs=NamedRerouteUsage(), Tex=NamedRerouteUsage()) {
   samplerSource=SSM_Clamp_WorldGroupSettings
   samplerType=SAMPLERTYPE_Masks
   texture=&T_WildEstate_BBOtto
  }
  n77 = Multiply(
   A=@n76.R
   B=SphereMask(
    A=NamedRerouteUsage()
    B=Constant() {r=0.5}
    Radius=ScalarParameter() {group=Otto_Card, parameterName=Plane_00_SphereMaskRadius}
    Hardness=ScalarParameter() {group=Otto_Card, parameterName=Plane_00_SphereMaskHardness}
   ) {attenuationRadius=0.5}
  )
  n78 = Saturate(None=Add(A=Add(A=@n74, B=@n75), B=@n77))
  n79 = ScalarParameter() {defaultValue=4.8552, group=Otto_Card, parameterName=Otto_Intensity}
  n80 = NamedRerouteUsage()
  NamedRerouteDeclaration(None=QualitySwitch(Default=Multiply(A=@n78, B=@n79), Low=@n80, Medium=@n80)) {name=OttoMan, nodeColor=<rgba 0.5469 1 0 1>}
  NamedRerouteDeclaration(None=@n71) {name=Bounds, nodeColor=<rgba 0 1 0.5781 1>}
  n81 = MaterialFunctionCall() {materialFunction=/Game/Athena/Aircraft/PartyBus_RT/Materials/MF_PartyBusAnimationWPO}
  NamedRerouteDeclaration(None=@n81.Result) {name=WPO, nodeColor=<rgba 1 0 0.1406 1>}
  n82 = Constant3Vector()
  n83 = MaterialFunctionCall(Float3=Divide(A=TransformPosition(None=@n82) {transformSourceType=TRANSFORMPOSSOURCE_Instance}) {constB=100}) {materialFunction=&BreakOutFloat3Components}
  n84 = Frac(None=Add(A=Add(A=@n83.R, B=@n83.R), B=@n83.B))
  n85 = Transform(None=Constant3Vector() {constant=<rgba 0 0 1 0>}) {transformSourceType=TRANSFORMSOURCE_Camera, transformType=TRANSFORM_Local}
  n86 = Transform(None=Constant3Vector() {constant=<rgba 1 0 0 0>}) {transformSourceType=TRANSFORMSOURCE_Camera, transformType=TRANSFORM_Local}
  n87 = ReflectionVectorWS()
  n88 = Add(
   A=Transform(None=ObjectPositionWS()) {transformSourceType=TRANSFORMSOURCE_World, transformType=TRANSFORM_Local}
   B=Multiply(A=Multiply(A=@n87) {constB=-1}, B=ScalarParameter() {defaultValue=0.1, parameterName=GlassRelectionOffset})
  )
  n89 = ScalarParameter() {defaultValue=0.5, parameterName=GlassRelectionTiling}
  NamedRerouteDeclaration(None=Add(
    A=Multiply(A=@n84) {constB=33.3}
    B=MaterialFunctionCall(Plane Normal Axis=@n85, Plane X Axis=@n86, Plane Center=@n88, UV Scale=@n89) {materialFunction=&VirtualPlaneCoordinates}.UVs
   )) {name=Reflection UVs, nodeColor=<rgba 0 0.5547 1 1>}
  NamedRerouteDeclaration(None=Multiply(A=Saturate(None=@n77), B=@n79)) {name=Front, nodeColor=<rgba 0 1 0.1562 1>}
  n90 = VectorParameter() {group=Otto_Card, parameterName=Otto_02_UVOffset, defaultValue=<rgba 0 0 0 1>}
  n91 = BumpOffset(
   Coordinate=Multiply(A=TextureCoordinate(), B=Constant2Vector() {g=1, r=-1})
   Height=ScalarParameter() {defaultValue=-2.5, group=Otto_Card, parameterName=Otto_02_BumpOffset}
   HeightRatioInput=Constant() {r=0.05}
  )
  n92 = ComponentMask(None=VectorParameter() {group=Otto_Card, parameterName=Otto_02_UVScale, defaultValue=<rgba 1 1 0 1>}.RGB) {g=true, r=true}
  NamedRerouteDeclaration(None=Add(A=ComponentMask(None=@n90.RGB) {g=true, r=true}, B=Multiply(A=@n91, B=@n92))) {name=Otto02UV, nodeColor=<rgba 1 0 0.1641 1>}
  OneMinus(None=Fresnel(ExponentIn=ScalarParameter() {defaultValue=1.4086, parameterName=FillFresnelEdge}, Normal=VertexNormalWS()) {baseReflectFraction=0})
  n93 = SetMaterialAttributes!(MaterialAttributes, Emissive Color=NamedRerouteUsage(), Opacity=NamedRerouteUsage()) {attributeSetTypes [#1, #2]}
  n94 = StaticSwitchParameter(
   True=@n93
   False=SetMaterialAttributes!(MaterialAttributes=@n93, World Position Offset=@n81.Result) {attributeSetTypes [#3]}
  ) {parameterName=UseDisableAllWPO}
  n95 = GetMaterialAttributes(None=@n94) {attributeGetTypes [#3]}
  n96 = MaterialFunctionCall!(
   WSDeformationCenter=VectorParameter() {bUseCustomPrimitiveData=true, group=z_code, parameterName=WSDeformationCenter, primitiveDataIndex=20}.RGB
   Code_DestructionAmount=ScalarParameter() {group=z_code, parameterName=Code_DestructionAmount}
   Code_UseUvForDamage
   Code_UseHitEffects=StaticBoolParameter() {defaultValue=true, group=z_code, parameterName=Code_UseHitEffects}
   Code_Emissive
   WSDamageRadius=ScalarParameter() {bUseCustomPrimitiveData=true, defaultValue=1, group=z_code, parameterName=WSDamageRadius, primitiveDataIndex=26}
   CrackDamageShader
   Diffuse
   ScalarModDamageCracks
   ScalarOffsetforDamageCracks
   ExposeLerpFunctionForInvertingTheHitDistanceCalc
   InvertDamageFalloff
   CrackWorldPosition
   WorldPosition
   DistanceCullFadeAnimationOffset=ScalarParameter() {bUseCustomPrimitiveData=true, group=z_code, parameterName=DistanceCullFadeAnimationOffset, primitiveDataIndex=12}
  ) {materialFunction=/Game/Packages/jl_tests/Fortnite_HitEffects}
  StaticSwitchParameter(
   True=SetMaterialAttributes(MaterialAttributes=@n95.MaterialAttributes, World Position Offset=Add(A=@n95.World Position Offset, B=@n96.Hit And Shrink WPO)) {attributeSetTypes [#3]}
   False=@n94
  ) {parameterName=UseDamageWobble}
  TextureCoordinate()
  TextureSample(UVs=NamedRerouteUsage()) {texture=&T_WildEstate_PartyBus_Windows}
  n97 = TransformPosition(None=Constant3Vector()) {transformSourceType=TRANSFORMPOSSOURCE_Instance}
  n98 = MaterialFunctionCall(Float3=Divide(A=@n97) {constB=100}) {materialFunction=&BreakOutFloat3Components}
  n99 = Add(A=@n98.R, B=@n98.R)
  n100 = Multiply(A=Frac(None=Add(A=@n99, B=@n98.B))) {constB=33.3}
  n101 = NamedRerouteUsage()
  n102 = Subtract(A=@n97, B=CameraPositionWS())
  n103 = ComponentMask(None=Transform(None=Normalize(VectorInput=@n102)) {transformSourceType=TRANSFORMSOURCE_World, transformType=TRANSFORM_Local}) {g=true, r=true}
  n104 = ScalarParameter() {defaultValue=0.1, parameterName=GlassRelectionOffset}
  TextureSample(UVs=Add(A=Add(A=@n100, B=@n101), B=Multiply(A=@n103, B=@n104))) {texture=&T_WildEstate_PartyBus_Windows}
  n105 = VertexNormalWS()
  TextureSample(UVs=MaterialFunctionCall!(
    "Normal (WS)"=Transform(None=@n105) {transformSourceType=TRANSFORMSOURCE_World, transformType=TRANSFORM_Local}
    Texture Scale=ScalarParameter() {defaultValue=0.5, parameterName=GlassRelectionTiling}
    Triplanar Fade
    Derivatives Scale
   ) {materialFunction=/Game/Athena/Environments/MaterialFunctions/MF_TriplanarDithered}.UVs) {texture=&T_WildEstate_PartyBus_Windows}
  TextureSample() {texture=&T_WildEstate_PartyBus_Windows}
  TextureSample(UVs=@n34) {texture=&T_WildEstate_PartyBus_Windows}
  VertexInterpolator(VS=PreSkinnedPosition())
 ]
}
