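# Material function for WildEstate base materials that bakes stepped, sun-facing shading into Base Color.
# It dots the world-space vertex normal with SunlightDirectionalVector from FortniteMaterialParameters, adds a bias, then quantizes the result into bands by rounding.
# Dark-bias and darkening lerps turn the bands into a Base Color multiplier.
# The UseSunDirectionShading static switch passes attributes through unchanged when off; a few leftover nodes are unconnected.
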MaterialFunction /WildEstate/Environment/BaseMaterials/MaterialFunctions/M_WildEstate_SunDirectionShading
graph {nodes [
  n1 = CollectionParameter() {
   collection=&FortniteMaterialParameters
   parameterName=SunlightDirectionalVector
  }
  n2 = DotProduct(A=VertexNormalWS(), B=ComponentMask(None=@n1) {b=true, g=true, r=true})
  Ceil(None=@n2)
  n3 = Add(A=@n2, B=ScalarParameter() {defaultValue=0.75, group=Shadin Sundirection, parameterName=SunDirection Shadow Bias})
  Divide(A=@n3)
  n4 = FunctionInput() {inputType=FunctionInput_MaterialAttributes}
  n5 = GetMaterialAttributes(None=@n4) {attributeGetTypes [#1]}
  n6 = ScalarParameter() {defaultValue=2, group=Shadin Sundirection, parameterName=SunDirection Shadow Steps}
  n7 = Multiply(A=@n3, B=@n6)
  n8 = LinearInterpolate(
   A=ScalarParameter() {defaultValue=-0.612, group=Shadin Sundirection, parameterName=SunDirection Shadow Dark Bias}
   Alpha=Divide(A=Round(None=@n7), B=@n6)
  )
  n9 = Multiply(
   A=@n5.BaseColor
   B=LinearInterpolate(A=ScalarParameter() {defaultValue=0.5, group=Shadin Sundirection, parameterName=SunDirection Shadow Darkening}, Alpha=Saturate(None=@n8))
  )
  FunctionOutput(None=StaticSwitchParameter(
    True=SetMaterialAttributes!(MaterialAttributes=@n5.MaterialAttributes, BaseColor=@n9) {attributeSetTypes [#1]}
    False=@n4
   ) {parameterName=UseSunDirectionShading})
  Normalize(VectorInput=AppendVector(A=ComponentMask(None=@n1) {g=true, r=true}, B=Constant() {r=-0.5}))
 ]}
