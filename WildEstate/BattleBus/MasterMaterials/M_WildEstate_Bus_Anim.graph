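# Master Battle Bus material for WildEstate, switching between quantized flat shading and the 2023 anime character shader.
# Optional second texture set, tint and time-of-day-masked emissive supply the surface inputs.
# A scrolling sign pans a 24-row flipbook text texture through a frame mask, brightened by time of day.
# Static switches layer world position offsets for bus animation, cloth, damage wobble, balloon bending and vertex-color-masked door rotation.
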Material /WildEstate/BattleBus/MasterMaterials/M_WildEstate_Bus_Anim
graph {
 properties {bUseMaterialAttributes=true, bUsedWithNanite=true, pixelDepthOffsetMode=PDOM_Legacy, usedShadingModels=Default Lit}
 nodes [
  n1 = VectorParameter() {parameterName=FresnelColor, defaultValue=<rgba 1 1 1 1>}
  n2 = Fresnel(ExponentIn=ScalarParameter() {defaultValue=1, group=Fresnel, parameterName=Fresnel_Exponent}, Normal=Transform(None=NamedRerouteUsage())) {baseReflectFraction=0}
  n3 = StaticSwitchParameter(True=Add(A=NamedRerouteUsage(), B=Multiply(A=@n1.RGB, B=@n2)), False=NamedRerouteUsage()) {parameterName=UseFresnel}
  n4 = NamedRerouteUsage()
  n5 = ScalarParameter() {group=FlatShading, parameterName=Specular_Override}
  n6 = ScalarParameter() {defaultValue=0.15, group=FlatShading, parameterName=Roughness_Override}
  n7 = NamedRerouteUsage()
  n8 = NamedRerouteUsage()
  n9 = TextureSampleParameter2D() {parameterName=OpacityMask, samplerType=SAMPLERTYPE_Masks, texture=&T_BlackMask}
  n10 = NamedRerouteUsage()
  n11 = NamedRerouteUsage()
  n12 = NamedRerouteUsage()
  n13 = ScalarParameter() {parameterName=ClearCoatRoughness}
  n14 = MaterialFunctionCall!(
   Material In=MaterialFunctionCall!(In=MakeMaterialAttributes!(
     BaseColor=@n3
     Metallic=@n4
     Specular=@n5
     Roughness=@n6
     Anisotropy
     EmissiveColor=@n7
     Opacity=@n8
     OpacityMask=@n9.G
     Normal=@n10
     Tangent
     WorldPositionOffset
     SubsurfaceColor=@n11
     ClearCoat=@n12
     ClearCoatRoughness=@n13
     AmbientOcclusion
     Refraction
     CustomizedUVs_0
     CustomizedUVs_1
     CustomizedUVs_2
     CustomizedUVs_3
     CustomizedUVs_4
     CustomizedUVs_5
     CustomizedUVs_6
     CustomizedUVs_7
     PixelDepthOffset
     ShadingModel
     Displacement
    )) {materialFunction=/WildEstate/Environment/BaseMaterials/MaterialFunctions/M_WildEstate_QantizeColor}.Result
   UseFlatShading=StaticBoolParameter() {group=FlatShading, parameterName=UseFlatShading_All, sortPriority=1}
   IsWorldSpace=StaticBool() {value=true}
   OverrideUnlitShadinBaseColorDarkening
   OverrideUnlitShadinBaseColorEmissive
  ) {materialFunction=/WildEstate/Environment/BaseMaterials/MaterialFunctions/MF_WildEstate_FlatShading}
  n15 = MaterialFunctionCall!(CustomUV) {materialFunction=/Game/Global/Materials/Character_Parent_2023/Anime_Parent_2023/MaterialFunctions/MFC_AnimeShader_2023_Pre}
  n16 = MaterialFunctionCall!(CustomLightVector, OverrideNormal_WS, STT=@n15.STT, Normal_Map=@n15.STT, DFL=@n15.STT, FX=@n15.STT) {materialFunction=/Game/Global/Materials/Character_Parent_2023/Anime_Parent_2023/MaterialFunctions/MFC_AnimeShader_2023_Calculations}
  n17 = StaticSwitchParameter(
   True=MaterialFunctionCall!(
    In=MaterialFunctionCall!(In=@n14.Result) {materialFunction=/WildEstate/Environment/BaseMaterials/MaterialFunctions/M_WildEstate_SunDirectionShading}.Result
    Enabled
    Disabled
    useDirectionalShading
    DirectionalShadingMask
   ) {materialFunction=/WildEstate/Environment/BaseMaterials/MaterialFunctions/MF_Test5050Unlit}.Result
   False=MaterialFunctionCall!(
    Shading_Gradient=@n16.Shading_Gradient
    Threshold_Gradient=@n16.Shading_Gradient
    Specular_Gradient=@n16.Shading_Gradient
    Sheen_Mask=@n16.Shading_Gradient
    Thickness=@n16.Shading_Gradient
    DFL_Mask=@n16.Shading_Gradient
    FX=@n16.Shading_Gradient
    Metallic_Gradient=@n16.Shading_Gradient
    RimLight_Gradient=@n16.Shading_Gradient
    NormalBuffer=@n16.Shading_Gradient
    --------
    CINE_LightSources=@n16.Shading_Gradient
    CINE_LightColor=@n16.Shading_Gradient
    ---------
    Color_Lit=@n15.Color_Lit
    Color_Shaded=@n15.Color_Lit
    SRM=@n15.Color_Lit
    Emissive=@n15.Color_Lit
    DFL_Color=@n15.Color_Lit
    Specular_Color=@n15.Color_Lit
    RimLight_Color=@n15.Color_Lit
   ) {materialFunction=/Game/Global/Materials/Character_Parent_2023/Anime_Parent_2023/MaterialFunctions/MFC_AnimeShader_2023_Post}.Result
  ) {group=00_ShadingModel, parameterName=UseWildEstate_BaseMaterial}
  n18 = GetMaterialAttributes(None=@n17) {attributeGetTypes [#1]}
  n19 = NamedRerouteUsage()
  Add(A=@n18.Emissive Color, B=@n19)
  n20 = ScalarParameter() {desc="0=left2right,1=right2left", group=ScrollingText, parameterName=PanDirection, sliderMax=1}
  n21 = Multiply(A=Time(), B=LinearInterpolate(A=Constant() {r=0.04}, B=Constant() {r=-0.04}, Alpha=Saturate(None=@n20)))
  n22 = ScalarParameter() {defaultValue=5, group=ScrollingText, parameterName=Frame steps}
  n23 = Multiply(A=@n21, B=@n22)
  Divide(A=Floor(None=Multiply(A=@n23, B=@n22)), B=@n22)
  n24 = NamedRerouteUsage()
  n25 = BlendMaterialAttributes(
   A=@n17
   B=SetMaterialAttributes(
    MaterialAttributes=@n18.MaterialAttributes
    Emissive Color=MaterialFunctionCall(Emissive=@n19) {materialFunction=&MF_FX_EmissiveAdjustments}.Result
   ) {attributeSetTypes [#1]}
   Alpha=NamedRerouteUsage()
  )
  n26 = StaticSwitchParameter(
   True=StaticSwitchParameter(
    True=SetMaterialAttributes!(MaterialAttributes=@n25, World Position Offset=@n24) {attributeSetTypes [#2]}
    False=@n25
   ) {desc=for debuging cloth anim, group=WPO_ClothAnimation, parameterName=UseWPO_ClothAnimation}
   False=SetMaterialAttributes(MaterialAttributes=@n25, World Position Offset=NamedRerouteUsage()) {attributeSetTypes [#2]}
  ) {parameterName=UseDisableAllWPO}
  n27 = GetMaterialAttributes(None=@n26) {attributeGetTypes [#2]}
  n28 = MaterialFunctionCall!(
   WSDeformationCenter=VectorParameter() {bUseCustomPrimitiveData=true, group=z_code, parameterName=WSDeformationCenter, primitiveDataIndex=20}.RGB
   Code_DestructionAmount=ScalarParameter() {group=z_code, parameterName=Code_DestructionAmount}
   Code_UseUvForDamage
   Code_UseHitEffects=StaticBoolParameter() {defaultValue=true, group=z_code, parameterName=Code_UseHitEffects}
   Code_Emissive
   WSDamageRadius=ScalarParameter() {bUseCustomPrimitiveData=true, defaultValue=1, group=z_code, parameterName=WSDamageRadius, primitiveDataIndex=26}
   CrackDamageShader
   Diffuse
   ScalarModDamageCracks
   ScalarOffsetforDamageCracks
   ExposeLerpFunctionForInvertingTheHitDistanceCalc
   InvertDamageFalloff
   CrackWorldPosition
   WorldPosition
   DistanceCullFadeAnimationOffset=ScalarParameter() {bUseCustomPrimitiveData=true, group=z_code, parameterName=DistanceCullFadeAnimationOffset, primitiveDataIndex=12}
  ) {materialFunction=/Game/Packages/jl_tests/Fortnite_HitEffects}
  n29 = StaticSwitchParameter(
   True=SetMaterialAttributes(MaterialAttributes=@n27.MaterialAttributes, World Position Offset=Add(A=@n27.World Position Offset, B=@n28.Hit And Shrink WPO)) {attributeSetTypes [#2]}
   False=@n26
  ) {parameterName=UseDamageWobble}
  n30 = GetMaterialAttributes(None=@n29) {attributeGetTypes [#2]}
  n31 = WorldPosition()
  n32 = ActorPositionWS()
  n33 = Add(A=ComponentMask(None=Subtract(A=@n31.XYZ, B=@n32)) {b=true}, B=Constant() {r=-500})
  n34 = Constant() {r=1400}
  n35 = Constant3Vector() {desc=BendDirection, constant=<rgba 0 -1 -0.65 1>}
  n36 = Constant() {r=1200}
  n37 = ScalarParameter() {parameterName=BalloonBendAmount, sliderMax=1}
  n38 = Multiply(
   A=Saturate(None=Divide(A=@n33, B=@n34))
   B=Multiply(A=Transform(None=@n35) {transformSourceType=TRANSFORMSOURCE_Local}, B=LinearInterpolate(B=@n36, Alpha=@n37))
  )
  n39 = StaticSwitchParameter(
   True=SetMaterialAttributes(MaterialAttributes=@n30.MaterialAttributes, World Position Offset=Add(A=@n30.World Position Offset, B=@n38)) {attributeSetTypes [#2]}
   False=@n29
  ) {parameterName=UseBendingBalloon}
  n40 = GetMaterialAttributes(None=@n39) {attributeGetTypes [#2]}
  n41 = VertexColor()
  n42 = ScalarParameter() {parameterName=DoorRotation}
  n43 = TransformPosition(None=VectorParameter() {parameterName=DoorPivotPoint, defaultValue=<rgba 0 -464.75 -50 1>}.RGB) {transformType=TRANSFORMPOSSOURCE_World}
  n44 = WorldPosition() {worldPositionShaderOffset=WPT_ExcludeAllShaderOffsets}
  n45 = Add(
   A=@n40.World Position Offset
   B=If(
    A=ChannelMaskParameter(None=@n41) {group=Door, parameterName=DoorOpenVertColor}.R
    B=Constant() {r=1}
    "A > B"=MaterialFunctionCall!(Rotation Amount=@n42, PivotPoint=@n43, WorldPosition=@n44.XYZ) {materialFunction=/Engine/Functions/Engine_MaterialFunctions02/WorldPositionOffset/RotateAboutWorldAxis_cheap}.X-Axis
    "A < B"=Constant()
   )
  )
  MaterialFunctionCall(MA In=StaticSwitchParameter(
    True=SetMaterialAttributes(MaterialAttributes=@n40.MaterialAttributes, World Position Offset=@n45) {attributeSetTypes [#2]}
    False=@n39
   ) {group=Door, parameterName=UseDoorWPO}) {materialFunction=&MF_CreativeMoveTool}
  Multiply(
   A=MaterialFunctionCall(Value=Time()) {materialFunction=&LinearSine}.Linear Sine
   B=ScalarParameter() {defaultValue=1, parameterName=SinMult}
  )
  n46 = TextureObject() {samplerType=SAMPLERTYPE_Masks, texture=/WildEstate/BattleBus/MasterMaterials/Textures/T_WildEstate_BB_Text_M}
  n47 = TextureSample(UVs=TextureCoordinate() {coordinateIndex=1}, Tex=@n46) {
   samplerSource=SSM_Clamp_WorldGroupSettings
   samplerType=SAMPLERTYPE_Masks
   texture=&T_WildEstate_BB_Text_M
  }
  n48 = TextureObjectParameter() {
   group=ScrollingText
   parameterName=Text
   samplerType=SAMPLERTYPE_LinearColor
   sortPriority=0
   texture=/WildEstate/BattleBus/Localization/T_WildEstate_BattleBus_Text_E
  }
  n49 = TextureCoordinate() {coordinateIndex=1}
  n50 = Constant() {r=0.5}
  n51 = Constant() {r=8}
  n52 = Add(
   A=Multiply(A=@n49, B=AppendVector(A=@n50, B=@n51)) {constB=10}
   B=ComponentMask(None=VectorParameter() {group=ScrollingText, parameterName=Offsets, defaultValue=<rgba 0 0.5 0 1>}.RGB) {g=true, r=true}
  )
  n53 = ComponentMask(None=@n52) {r=true}
  n54 = StaticSwitchParameter(True=AppendVector(A=OneMinus(None=@n53), B=ComponentMask(None=@n52) {g=true}), False=@n52) {group=ScrollingText, parameterName=Flip}
  n55 = Constant3Vector() {constant=<rgba 2 0 0 0>}
  n56 = VectorParameter() {group=ScrollingText, parameterName=Sign_Color, defaultValue=<rgba 7 1.7719 0 1>}
  n57 = Constant() {r=0.5}
  n58 = MaterialFunctionCall(Value_Morning=@n57, Value_Day=Constant() {r=1}, Value_Evening=@n57, Value_Night=Constant() {r=0.33}) {materialFunction=&Scalar_Time_of_Day}
  n59 = DotProduct(A=CameraVectorWS(), B=VertexNormalWS())
  n60 = Multiply(
   A=MaterialFunctionCall(
    "Animation  Phase (0-1)"=ScalarParameter() {defaultValue=0.1, desc=shuffles 24 lines based on 0-1, group=ScrollingText, parameterName=TextSelection, sliderMax=1}
    Number of Rows=Constant() {r=24}
    Number of Columns=Constant() {r=1}
    Texture=@n48
    UVs=Panner(Coordinate=@n54, Time=@n21, Speed=@n55) {speedX=-0.5}
   ) {materialFunction=&FlipBook}.Result
   B=LinearInterpolate(A=Constant3Vector() {constant=<rgba 0.7043 0.0894 0 1>}, B=Multiply(A=@n56.RGB, B=@n58.Result), Alpha=Power(Base=@n59) {constExponent=4})
  )
  Multiply(A=NamedRerouteDeclaration(None=@n47.G) {name=EmissiveMask, nodeColor=<rgba 0.5469 1 0 1>}, B=@n60)
  n61 = MaterialFunctionCall() {materialFunction=/Game/Athena/Aircraft/PartyBus_RT/Materials/MF_PartyBusAnimationWPO}
  n62 = MaterialFunctionCall!(
   UVs
   Pan Speed
   WPO_ClothOffsetDirection
   WPO_ClothNormalIntensity
   WPO_ClothOffsetIntensity
   "Motion Mask (Defaults to the Red Vertex Color)"=ChannelMaskParameter(None=VertexColor()) {group=WPO_ClothAnimation, maskChannel=Blue, parameterName=ClothMask, defaultValue=<rgba 0 0 1 0>}.R
   WPO_ClothWPOWindTextureOffset
   Pan Direction
  ) {materialFunction=/Game/Packages/Fortress_SharedMaterials/Base_Material_Functions/Cloth}
  NamedRerouteDeclaration(None=StaticSwitchParameter(True=Add(A=@n61.Result, B=@n62.Additive WPO), False=@n61.Result) {group=WPO_ClothAnimation, parameterName=UseWPO_ClothAnimation}) {name=WPO, nodeColor=<rgba 1 0 0.3516 1>}
  NamedRerouteDeclaration(None=@n62.Additive WPO) {name=ClothWPO, nodeColor=<rgba 1 0 0.4922 1>}
  n63 = TextureSampleParameter2D() {group=Textures, parameterName=M, samplerType=SAMPLERTYPE_Masks, texture=&T_BlackMask}
  n64 = TextureSampleParameter2D() {group=Textures2, parameterName=M2, samplerType=SAMPLERTYPE_Masks, texture=&T_BlackMask}
  n65 = StaticSwitchParameter(True=LinearInterpolate(A=@n63.R, B=@n64.R, Alpha=NamedRerouteUsage()), False=@n63.R) {parameterName=Use2ndMaterial}
  n66 = Multiply(A=@n65, B=ScalarParameter() {defaultValue=1, parameterName=ClearCoat})
  NamedRerouteDeclaration(None=@n66) {name=ClearCoat, nodeColor=<rgba 0 1 0.3438 1>}
  n67 = ScalarParameter() {defaultValue=1, parameterName=SSS_Opacity}
  n68 = NamedRerouteUsage()
  n69 = StaticSwitchParameter(True=LinearInterpolate(A=@n67, B=ScalarParameter() {defaultValue=1, parameterName=SSS_Opacity2}, Alpha=@n68), False=@n67) {parameterName=Use2ndMaterial}
  n70 = StaticSwitchParameter(True=Multiply(A=@n65, B=@n69), False=@n69) {defaultValue=true, parameterName=UseMtexture}
  NamedRerouteDeclaration(None=@n70) {name=Opacity, nodeColor=<rgba 1 0 0.9141 1>}
  n71 = TextureSampleParameter2D() {
   group=Textures
   parameterName=SpecularMask
   samplerType=SAMPLERTYPE_Masks
   texture=/Game/Athena/Aircraft/PartyBus_RT/Textures/T_PartyBus_RT_Balloon_S
  }
  n72 = TextureSampleParameter2D() {
   group=Textures2
   parameterName=SpecularMask2
   samplerType=SAMPLERTYPE_Masks
   texture=&T_PartyBus_RT_Balloon_S
  }
  n73 = NamedRerouteUsage()
  n74 = MaterialFunctionCall(Float3=StaticSwitchParameter(True=LinearInterpolate(A=@n71.RGB, B=@n72.RGB, Alpha=@n73), False=@n71.RGB) {parameterName=Use2ndMaterial}) {materialFunction=&BreakOutFloat3Components}
  n75 = StaticSwitchParameter(
   True=LinearInterpolate(A=ScalarParameter() {parameterName=Specular_1}, B=ScalarParameter() {defaultValue=1, parameterName=Specular_2}, Alpha=@n74.R)
   False=@n74.R
  ) {parameterName=UseModifySpecular}
  NamedRerouteDeclaration(None=@n75) {name=Specular, nodeColor=<rgba 0 1 0.0156 1>}
  n76 = StaticSwitchParameter(
   True=LinearInterpolate(A=ScalarParameter() {parameterName=Metal_Low}, B=ScalarParameter() {defaultValue=1, parameterName=Metal_High}, Alpha=@n74.G)
   False=@n74.G
  ) {parameterName=UseModifyMetal}
  NamedRerouteDeclaration(None=@n76) {name=Metallic, nodeColor=<rgba 0.6172 0 1 1>}
  n77 = StaticSwitchParameter(
   True=LinearInterpolate(A=ScalarParameter() {parameterName=Roughness_Low}, B=ScalarParameter() {defaultValue=1, parameterName=Roughness_High}, Alpha=@n74.B)
   False=@n74.B
  ) {parameterName=UseModifyRoughness}
  NamedRerouteDeclaration(None=@n77) {name=Roughness, nodeColor=<rgba 0 0.0859 1 1>}
  n78 = ScalarParameter() {defaultValue=1, parameterName=SSS_Brightness}
  n79 = ScalarParameter() {defaultValue=1, parameterName=SSS_Brightness2}
  n80 = Multiply(
   A=NamedRerouteUsage()
   B=StaticSwitchParameter(True=LinearInterpolate(A=@n78, B=@n79, Alpha=@n68), False=@n78) {parameterName=Use2ndMaterial}
  )
  NamedRerouteDeclaration(None=@n80) {name=SubSurfaceColor, nodeColor=<rgba 0 0.6016 1 1>}
  n81 = ScalarParameter() {defaultValue=6.1, group=ScrollingText, parameterName=Frame Mask Scale}
  n82 = Multiply(A=TextureCoordinate(), B=QualitySwitch(Default=@n81, Low=Multiply(A=@n81) {constB=2})) {constB=4}
  n83 = ScalarParameter() {defaultValue=2, group=ScrollingText, parameterName=MorningEvening}
  n84 = MaterialFunctionCall(
   Value_Morning=@n83
   Value_Day=ScalarParameter() {defaultValue=3, group=ScrollingText, parameterName=Day}
   Value_Evening=@n83
   Value_Night=ScalarParameter() {defaultValue=10, group=ScrollingText, parameterName=Night}
  ) {materialFunction=&Scalar_Time_of_Day}
  NamedRerouteDeclaration(None=Multiply(
    A=TextureSample(UVs=@n82, Tex=@n46) {samplerType=SAMPLERTYPE_Masks, texture=&T_WildEstate_BB_Text_M}.R
    B=Multiply(A=@n60, B=@n84.Result)
   )) {name=EmissiveText, nodeColor=<rgba 0 0.0625 1 1>}
  NamedRerouteDeclaration(None=@n62.Normal) {name=ClothNormal, nodeColor=<rgba 0 1 0.9531 1>}
  n85 = ScalarParameter() {defaultValue=-0.7021, parameterName=low}
  n86 = ScalarParameter() {defaultValue=2, parameterName=high}
  n87 = Constant() {r=0.5}
  n88 = MaterialFunctionCall!(
   Sine Phase=Divide(
    A=CollectionParameter() {collection=&FortniteMaterialParameters, parameterName=TimeOfDay_0-1}
    B=MaterialFunctionCall(Multiplier=@n87) {materialFunction=&Pi}.Result
   ) {constB=24}
   Value 1=Constant() {r=1}
   Value 2=Constant()
  ) {materialFunction=/Engine/Functions/Engine_MaterialFunctions02/Utility/Sine_Remapped}
  NamedRerouteDeclaration(None=Clamp(None=LinearInterpolate(A=@n85, B=@n86, Alpha=@n88.Remapped Sine))) {name=TimeOfDayMask, nodeColor=<rgba 0.5703 0 1 1>}
  n89 = ComponentMask(None=TextureCoordinate() {coordinateIndex=1}) {r=true}
  n90 = Floor(None=Subtract(A=@n89))
  n91 = Multiply(A=Saturate(None=Floor(None=@n89)), B=OneMinus(None=Saturate(None=@n90)))
  NamedRerouteDeclaration(None=@n91) {name=2ndUVChannelMask, nodeColor=<rgba 0 0.5312 1 1>}
  n92 = TextureSampleParameter2D() {group=Textures, parameterName=Emissive, texture=/Game/Athena/Aircraft/PartyBus_RT/Textures/T_PartyBus_RT_Balloon_E}
  n93 = StaticSwitchParameter(
   True=LinearInterpolate(
    A=@n92.RGB
    B=TextureSampleParameter2D() {group=Textures2, parameterName=Emissive2, texture=&T_PartyBus_RT_Balloon_E}.RGB
    Alpha=NamedRerouteUsage()
   )
   False=@n92.RGB
  ) {parameterName=Use2ndMaterial}
  n94 = MaterialFunctionCall(Float3=@n93) {materialFunction=&BreakOutFloat3Components}
  n95 = Add(
   A=Multiply(A=VectorParameter() {parameterName=EmissiveColorRedChannel, defaultValue=<rgba 0 0 0 1>}.RGB, B=@n94.R)
   B=Multiply(A=VectorParameter() {parameterName=EmissiveColorBlueChannel, defaultValue=<rgba 0 0 0 1>}.RGB, B=@n94.B)
  )
  n96 = Multiply(
   A=ScalarParameter() {defaultValue=1, parameterName=EmissiveBrightness}
   B=Power(
    Base=StaticSwitchParameter(True=@n95, False=@n93) {parameterName=UseEmissiveColorChannels}
    Exp=ScalarParameter() {defaultValue=1, parameterName=EmissiveContrast}
   )
  )
  n97 = Multiply(A=@n96, B=NamedRerouteUsage())
  NamedRerouteDeclaration(None=StaticSwitchParameter(True=StaticSwitchParameter(True=@n97, False=@n96) {defaultValue=true, parameterName=UseTODEmissiveMask}, False=Constant()) {defaultValue=true, parameterName=UseEmissive}) {name=Emissive, nodeColor=<rgba 0 0.25 1 1>}
  n98 = TextureSample(UVs=TextureCoordinate() {coordinateIndex=1}) {texture=/Game/Athena/Aircraft/PartyBus_Birthday/Textures/T_Partybus_Birthday_Colors7}
  n99 = ScalarParameter() {group=ColorCorrection, parameterName=BaseColorHueShift}
  n100 = TextureSampleParameter2D() {group=Textures, parameterName=BaseColor, texture=/Game/Athena/Aircraft/PartyBus_RT/Textures/T_PartyBus_RT_Balloon_D}
  n101 = StaticSwitchParameter(
   True=LinearInterpolate(
    A=@n100.RGB
    B=TextureSampleParameter2D() {group=Textures2, parameterName=BaseColor2, texture=&T_PartyBus_RT_Balloon_D}.RGB
    Alpha=NamedRerouteUsage()
   )
   False=@n100.RGB
  ) {parameterName=Use2ndMaterial}
  n102 = StaticSwitchParameter(
   True=Multiply(
    A=VectorParameter() {group=ColorCorrection, parameterName=BaseColorTint, defaultValue=<rgba 1 1 1 1>}.RGB
    B=MaterialFunctionCall(Hue Shift Percentage=@n99, Texture=@n101) {materialFunction=&HueShift}.Result
   )
   False=@n101
  ) {parameterName=UseColorCorrection}
  NamedRerouteDeclaration(None=StaticSwitchParameter(True=Multiply(A=@n98.RGB, B=@n102), False=@n102) {parameterName=UseColorization}) {name=BaseColor, nodeColor=<rgba 0.0547 0 1 1>}
  NamedRerouteDeclaration(None=TextureSampleParameter2D() {parameterName=FX_Mask, samplerType=SAMPLERTYPE_Masks, texture=&T_BlackMask}.RGBA) {name=SkinFX_Mask, nodeColor=<rgba 0 1 0.8828 1>}
  n103 = TextureSampleParameter2D() {
   group=Textures
   parameterName=Normal
   samplerType=SAMPLERTYPE_Normal
   texture=/Game/Athena/Aircraft/PartyBus_RT/Textures/T_PartyBus_RT_Balloon_N
  }
  n104 = TextureSampleParameter2D() {
   group=Textures2
   parameterName=Normal2
   samplerType=SAMPLERTYPE_Normal
   texture=&T_PartyBus_RT_Balloon_N
  }
  n105 = StaticSwitchParameter(True=LinearInterpolate(A=@n103.RGB, B=@n104.RGB, Alpha=NamedRerouteUsage()), False=@n103.RGB) {parameterName=Use2ndMaterial}
  n106 = NamedRerouteUsage()
  n107 = StaticSwitchParameter(
   True=ShadingPathSwitch(
    Default=MaterialFunctionCall(BaseNormal=@n105, AdditionalNormal=@n106) {materialFunction=&BlendAngleCorrectedNormals}.Result
    Mobile=@n105
   ) {desc=Migrated from FeatureLevelSwitch}
   False=@n105
  ) {group=WPO_ClothAnimation, parameterName=UseWPO_ClothAnimation}
  NamedRerouteDeclaration(None=@n107) {name=Normal, nodeColor=<rgba 0 0.0859 1 1>}
  NamedRerouteDeclaration(None=@n105) {name=NormalTextures, nodeColor=<rgba 0.125 0 1 1>}
  NamedRerouteUsage()
  NamedRerouteUsage()
  QualitySwitch(
   Default=@n49
   Low=MaterialFunctionCall(UVs=@n49, "Rotation Angle (0-1)"=Constant() {r=0.5}) {materialFunction=&CustomRotator}.Rotated Values
  )
  Reroute(None=@n91)
  Reroute(None=@n107)
  Reroute(None=@n105)
  Reroute(None=@n66)
  Reroute(None=@n70)
  Reroute(None=@n75)
  Reroute(None=@n76)
  Reroute(None=@n77)
  Reroute(None=@n80)
  ShadingPathSwitch(Default=@n28.Hit And Shrink WPO) {desc=Migrated from FeatureLevelSwitch}
 ]
}
